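# Figment's default bullet whip MetaSound, a quad one-shot for bullet pass-bys.
# It plays a random whip wave with varied pitch and gain, scaled by an affiliation mix of enemy and friendly modulation buses.
# An envelope crossfades front to side, alignment weights left/right, and an azimuth panner takes distance as gain.
# OnStop fades out after distance checks involving azimuth; the Mobile page fades directly and links only front outputs.
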MetaSoundSource /FigmentAudioContent/Content/Whips/MSS_BulletWhip_Figment_Default
inputs [
 BulletWhip.FadeOutTime {type=Time}
 BulletWhip.IgnoreImpactDistance {type=Float}
 BulletWhip.OnStop {type=Trigger}
 BulletWhip.PassByAlignment {type=Float}
 BulletWhip.PassByAzimuth {type=Float}
 BulletWhip.PassByCrossfadeTime {type=Time}
 BulletWhip.PassByCurvePower {type=Float}
 BulletWhip.PassByDistance {type=Float}
 Character.Affiliation {type=Enum:FortAudioAffiliation}
 Character.Health {type=Float}
 Character.UniqueID {type=Int32}
 FriendlyMultiplier {type=Float}
 Sounds {type=WaveAsset:Array}
 UE.Source.OnPlay {type=Trigger}
]
interfaces [Fort.BulletWhip:1.0, Fort.Character:1.0, UE.OutputFormat.Quad:1.0, UE.Source.OneShot:1.0, UE.Source:1.1]
outputs [
 "UE.OutputFormat.Quad.Audio:0" {type=Audio}
 "UE.OutputFormat.Quad.Audio:1" {type=Audio}
 "UE.OutputFormat.Quad.Audio:4" {type=Audio}
 "UE.OutputFormat.Quad.Audio:5" {type=Audio}
 UE.Source.OneShot.OnFinished {type=Trigger}
]
pages [
 Default {
  is_build_graph=true
  input_defaults {
   BulletWhip.FadeOutTime=0.1
   BulletWhip.IgnoreImpactDistance=0.1
   BulletWhip.OnStop=null
   BulletWhip.PassByAlignment=0
   BulletWhip.PassByAzimuth=-177.12
   BulletWhip.PassByCrossfadeTime=0.2
   BulletWhip.PassByCurvePower=4
   BulletWhip.PassByDistance=0.04
   Character.Affiliation=0
   Character.Health=0
   Character.UniqueID=0
   FriendlyMultiplier=1
   UE.Source.OnPlay=null
   Sounds [
    /FigmentAudioContent/Content/Whips/SoundWavs/bullet_whip_01
    /FigmentAudioContent/Content/Whips/SoundWavs/bullet_whip_02
    /FigmentAudioContent/Content/Whips/SoundWavs/bullet_whip_03
    /FigmentAudioContent/Content/Whips/SoundWavs/bullet_whip_04
   ]
  }
  nodes [
   n1 = Modulation.GetModulatorValue() {also {defaults {Modulator=/Game/Sounds/Modulation/Buses/CB_Weapon_Whip_Friendly}}}
   n2 = TriggerCompare.Clamp:Int32(Compare=^UE.Source.OnPlay, A=Array.Num:WaveAsset:Array(Array=^Sounds).Num) {also {defaults {B=0}}}
   ArrayRandomGet_0 = Array.Random Get:WaveAsset:Array(In Array=^Sounds, Next=@n2.False) {also {defaults {Enable Shared State="<MetaSoundBuilderResult.SUCCEEDED: 0>", Weights []}}}
   n3 = UE.RandomFloat(Next=@ArrayRandomGet_0.On Next) {also {defaults {Max=1.2, Min=0.9}}}
   n4 = UE.RandomFloat(Next=@n3.On Next) {also {defaults {Min=0.9}}}
   FrequencyMultipliertoSemitone_0 = UE.Frequency Multiplier to Semitone:Float(Frequency Multiplier=@n3.Value)
   WavePlayer_0 = UE.Wave Player:Mono(Play=@ArrayRandomGet_0.On Next, Wave Asset=@ArrayRandomGet_0.Value, Pitch Shift=@FrequencyMultipliertoSemitone_0.Semitones) {also {version [1,1]}}
   n5 = InitVariable.Float()
   n6 = MapRange.MapRange:Float(In=AbsoluteValue.Abs:Float(Input=^BulletWhip.PassByAzimuth).Value) {also {defaults {In Range A=175, In Range B=176, Out Range B=0.8}}}
   n7 = TriggerCompare.Clamp:Float(
    Compare=^BulletWhip.OnStop
    A=^BulletWhip.PassByDistance
    B=VariableAccessor.Float(Variable=VariableMutator.Float(Variable=@n5.Variable, Value=@n6.Out Value).Variable).Value
   ) {also {defaults {Type=5}}}
   n8 = Value.Value:Float(Target Value=^BulletWhip.IgnoreImpactDistance, Reset=@n7.False, Set=@n7.True) {also {defaults {Init Value=0.1}}}
   TriggerAnyTriggerAny2_0 = "TriggerAny.Trigger Any (2)"(In 1=@n8.On Reset, In 0=@n8.On Set)
   ADEnvelopeADEnvelope_1 = AD Envelope.AD Envelope:Float(
    Decay Time=^BulletWhip.FadeOutTime
    Trigger=TriggerCompare.Clamp:Float(A=^BulletWhip.PassByDistance, Compare=@TriggerAnyTriggerAny2_0.Out, B=@n8.Output Value) {also {defaults {Type=5}}}.True
   ) {also {defaults {Attack Time=0}}}
   TriggerRouteTriggerRouteFloat2_0 = "TriggerRoute.Trigger Route (Float, 2):Float"(Set 0=^UE.Source.OnPlay, Set 1=@ADEnvelopeADEnvelope_1.On Trigger, Value 1=@ADEnvelopeADEnvelope_1.Out Envelope) {also {defaults {Value 0=1}}}
   n9 = UE.Multiply:Audio by Float(
    AdditionalOperands=&MSP_AffiliationMixFloat(
     Affiliation=^Character.Affiliation
     Enemy=Modulation.GetModulatorValue() {also {defaults {Modulator=/Game/Sounds/Modulation/Buses/CB_Weapon_Whip_Enemy}}}.Out
     Non-Enemy=UE.Multiply:Float(PrimaryOperand=^FriendlyMultiplier, AdditionalOperands=@n1.Out).Out
    ).Out
    PrimaryOperand=UE.Multiply:Audio by Float(
     PrimaryOperand=UE.Multiply:Audio by Float(AdditionalOperands=@n4.Value, PrimaryOperand=@WavePlayer_0.Out Mono).Out
     AdditionalOperands=@TriggerRouteTriggerRouteFloat2_0.Value
    ).Out
   )
   ADEnvelopeADEnvelope_0 = AD Envelope.AD Envelope:Float(Decay Time=^BulletWhip.PassByCrossfadeTime, Decay Curve=^BulletWhip.PassByCurvePower, Trigger=^UE.Source.OnPlay) {also {defaults {Attack Time=0}}}
   n10 = VariableAccessor.Float(Variable=VariableMutator.Float(
     Variable=InitVariable.Float().Variable
     Value=UE.Subtract:Float(AdditionalOperands=@ADEnvelopeADEnvelope_0.Out Envelope) {also {defaults {PrimaryOperand=1}}}.Out
    ).Variable)
   n11 = VariableAccessor.Float(Variable=InitVariable.Float().Variable)
   n12 = UE.Multiply:Float(PrimaryOperand=@n11.Value) {also {defaults {AdditionalOperands=-1}}}
   n13 = VariableAccessor.Float(Variable=VariableMutator.Float(
     Variable=InitVariable.Float().Variable
     Value=MapRange.MapRange:Float(In=^BulletWhip.PassByAlignment, In Range B=@n12.Out) {also {defaults {Out Range A=1, Out Range B=0}}}.Out Value
    ).Variable)
   n14 = VariableAccessor.Float(Variable=VariableMutator.Float(Value=@ADEnvelopeADEnvelope_0.Out Envelope, Variable=InitVariable.Float().Variable).Variable)
   n15 = VariableAccessor.Float(Variable=@n14.Variable)
   n16 = VariableAccessor.Float(Variable=VariableMutator.Float(
     Variable=InitVariable.Float().Variable
     Value=MapRange.MapRange:Float(In=^BulletWhip.PassByAlignment, In Range B=@n11.Value) {also {defaults {Out Range A=1, Out Range B=0}}}.Out Value
    ).Variable)
   n17 = VariableAccessor.Float(Variable=@n10.Variable)
   MSP_AzimuthPannerQuad_0 = &MSP_AzimuthPannerQuad(
    Azimuth=^BulletWhip.PassByAzimuth
    Gain=^BulletWhip.PassByDistance
    In Side Left=UE.Multiply:Audio by Float(PrimaryOperand=UE.Multiply:Audio by Float(PrimaryOperand=@n9.Out, AdditionalOperands=@n10.Value).Out, AdditionalOperands=@n13.Value).Out
    In Front Right=UE.Multiply:Audio by Float(PrimaryOperand=UE.Multiply:Audio by Float(PrimaryOperand=@n9.Out, AdditionalOperands=@n15.Value).Out, AdditionalOperands=@n16.Value).Out
    In Front Left=UE.Multiply:Audio by Float(PrimaryOperand=UE.Multiply:Audio by Float(PrimaryOperand=@n9.Out, AdditionalOperands=@n14.Value).Out, AdditionalOperands=@n13.Value).Out
    In Side Right=UE.Multiply:Audio by Float(PrimaryOperand=UE.Multiply:Audio by Float(PrimaryOperand=@n9.Out, AdditionalOperands=@n17.Value).Out, AdditionalOperands=@n16.Value).Out
   )
   GetWaveDuration_0 = "UE.Get Wave Duration: "(Wave=@ArrayRandomGet_0.Value)
   TriggerDelay_0 = UE.Trigger Delay(In=@ArrayRandomGet_0.On Next, Delay Time=@GetWaveDuration_0.Duration)
   TriggerAnyTriggerAny4_0 = "TriggerAny.Trigger Any (4)"(In 3=@ADEnvelopeADEnvelope_1.On Done, In 0=@n2.True, In 2=@TriggerDelay_0.Out, In 1=@WavePlayer_0.On Finished)
  ]
  links [
   {dst="@UE.OutputFormat.Quad.Audio:0", src=MSP_AzimuthPannerQuad_0.Out Front Left}
   {dst="@UE.OutputFormat.Quad.Audio:1", src=MSP_AzimuthPannerQuad_0.Out Front Right}
   {dst="@UE.OutputFormat.Quad.Audio:4", src=MSP_AzimuthPannerQuad_0.Out Side Left}
   {dst="@UE.OutputFormat.Quad.Audio:5", src=MSP_AzimuthPannerQuad_0.Out Side Right}
   {dst="@UE.Source.OneShot.OnFinished", src=TriggerAnyTriggerAny4_0.Out}
  ]
 }
 Mobile {
  nodes [
   InitVariable.Audio()
   InitVariable.Float()
   InitVariable.Float()
   n1 = Modulation.GetModulatorValue() {also {defaults {Modulator=&CB_Weapon_Whip_Friendly}}}
   n2 = TriggerCompare.Clamp:Int32(Compare=^UE.Source.OnPlay, A=Array.Num:WaveAsset:Array(Array=^Sounds).Num) {also {defaults {B=0}}}
   ArrayRandomGet_0 = Array.Random Get:WaveAsset:Array(In Array=^Sounds, Next=@n2.False) {also {defaults {Enable Shared State="<MetaSoundBuilderResult.SUCCEEDED: 0>", Weights []}}}
   n3 = UE.RandomFloat(Next=@ArrayRandomGet_0.On Next) {also {defaults {Max=1.2, Min=0.9}}}
   n4 = UE.RandomFloat(Next=@n3.On Next) {also {defaults {Min=0.9}}}
   FrequencyMultipliertoSemitone_0 = UE.Frequency Multiplier to Semitone:Float(Frequency Multiplier=@n3.Value)
   WavePlayer_0 = UE.Wave Player:Mono(Play=@ArrayRandomGet_0.On Next, Wave Asset=@ArrayRandomGet_0.Value, Pitch Shift=@FrequencyMultipliertoSemitone_0.Semitones) {also {version [1,1]}}
   ADEnvelopeADEnvelope_1 = AD Envelope.AD Envelope:Float(Decay Time=^BulletWhip.FadeOutTime, Trigger=^BulletWhip.OnStop) {also {defaults {Attack Time=0}}}
   TriggerRouteTriggerRouteFloat2_0 = "TriggerRoute.Trigger Route (Float, 2):Float"(Set 0=^UE.Source.OnPlay, Set 1=@ADEnvelopeADEnvelope_1.On Trigger, Value 1=@ADEnvelopeADEnvelope_1.Out Envelope) {also {defaults {Value 0=1}}}
   n5 = UE.Multiply:Audio by Float(
    AdditionalOperands=&MSP_AffiliationMixFloat(
     Affiliation=^Character.Affiliation
     Enemy=Modulation.GetModulatorValue() {also {defaults {Modulator=&CB_Weapon_Whip_Enemy}}}.Out
     Non-Enemy=UE.Multiply:Float(PrimaryOperand=^FriendlyMultiplier, AdditionalOperands=@n1.Out).Out
    ).Out
    PrimaryOperand=UE.Multiply:Audio by Float(
     PrimaryOperand=UE.Multiply:Audio by Float(AdditionalOperands=@n4.Value, PrimaryOperand=@WavePlayer_0.Out Mono).Out
     AdditionalOperands=@TriggerRouteTriggerRouteFloat2_0.Value
    ).Out
   )
   ADEnvelopeADEnvelope_0 = AD Envelope.AD Envelope:Float(Decay Time=^BulletWhip.PassByCrossfadeTime, Decay Curve=^BulletWhip.PassByCurvePower, Trigger=^UE.Source.OnPlay) {also {defaults {Attack Time=0}}}
   n6 = VariableAccessor.Float(Variable=VariableMutator.Float(
     Variable=InitVariable.Float().Variable
     Value=UE.Subtract:Float(AdditionalOperands=@ADEnvelopeADEnvelope_0.Out Envelope) {also {defaults {PrimaryOperand=1}}}.Out
    ).Variable)
   n7 = VariableAccessor.Float(Variable=InitVariable.Float().Variable)
   n8 = UE.Multiply:Float(PrimaryOperand=@n7.Value) {also {defaults {AdditionalOperands=-1}}}
   n9 = VariableAccessor.Float(Variable=VariableMutator.Float(
     Variable=InitVariable.Float().Variable
     Value=MapRange.MapRange:Float(In=^BulletWhip.PassByAlignment, In Range B=@n8.Out) {also {defaults {Out Range A=1, Out Range B=0}}}.Out Value
    ).Variable)
   n10 = VariableAccessor.Float(Variable=VariableMutator.Float(Value=@ADEnvelopeADEnvelope_0.Out Envelope, Variable=InitVariable.Float().Variable).Variable)
   n11 = VariableAccessor.Float(Variable=@n10.Variable)
   n12 = VariableAccessor.Float(Variable=VariableMutator.Float(
     Variable=InitVariable.Float().Variable
     Value=MapRange.MapRange:Float(In=^BulletWhip.PassByAlignment, In Range B=@n7.Value) {also {defaults {Out Range A=1, Out Range B=0}}}.Out Value
    ).Variable)
   n13 = VariableAccessor.Float(Variable=@n6.Variable)
   MSP_AzimuthPannerQuad_0 = &MSP_AzimuthPannerQuad(
    Azimuth=^BulletWhip.PassByAzimuth
    Gain=^BulletWhip.PassByDistance
    In Side Left=UE.Multiply:Audio by Float(PrimaryOperand=UE.Multiply:Audio by Float(PrimaryOperand=@n5.Out, AdditionalOperands=@n6.Value).Out, AdditionalOperands=@n9.Value).Out
    In Front Right=UE.Multiply:Audio by Float(PrimaryOperand=UE.Multiply:Audio by Float(PrimaryOperand=@n5.Out, AdditionalOperands=@n11.Value).Out, AdditionalOperands=@n12.Value).Out
    In Front Left=UE.Multiply:Audio by Float(PrimaryOperand=UE.Multiply:Audio by Float(PrimaryOperand=@n5.Out, AdditionalOperands=@n10.Value).Out, AdditionalOperands=@n9.Value).Out
    In Side Right=UE.Multiply:Audio by Float(PrimaryOperand=UE.Multiply:Audio by Float(PrimaryOperand=@n5.Out, AdditionalOperands=@n13.Value).Out, AdditionalOperands=@n12.Value).Out
   )
   TriggerAnyTriggerAny3_0 = "TriggerAny.Trigger Any (3)"(In 2=@ADEnvelopeADEnvelope_1.On Done, In 0=@n2.True, In 1=@WavePlayer_0.On Finished)
  ]
  links [
   {dst="@UE.OutputFormat.Quad.Audio:0", src=MSP_AzimuthPannerQuad_0.Out Front Left}
   {dst="@UE.OutputFormat.Quad.Audio:1", src=MSP_AzimuthPannerQuad_0.Out Front Right}
   {dst="@UE.Source.OneShot.OnFinished", src=TriggerAnyTriggerAny3_0.Out}
  ]
 }
]
properties {
 assetImportData=&self:AssetImportData
 numChannels=4
 outputFormat=Quad
 sampleRate=48000
 soundClassObject=&Fort_WeaponRanged_Projectile_Effects
 volume=0.9
 concurrencySet [/Game/Sounds/Concurrency_Rules/Athena_BulletWhizBy_Conc]
 modulationSettings {
  volumeRouting=Union
  volumeModulationDestination {bEnableModulation=true, modulators [/Game/Sounds/Modulation/EnvelopeFollowers/FortWeapon/MP_WeaponFireEnvFollower]}
 }
}
